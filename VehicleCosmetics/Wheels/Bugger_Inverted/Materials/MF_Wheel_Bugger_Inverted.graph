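# Material function for the Bugger Inverted vehicle wheel cosmetic that outputs material attributes with an animated emissive color.
# The rim branch combines custom color and emission with mask-gated, sine-driven pulsing and a fast-scrolling gradient stripe limited by rim masks.
# The tire branch tints static masks and a horizontally panning tread texture, shaped by a linear gradient, with custom color and emission.
# Vertex color red blends between the rim and tire results.
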MaterialFunction /VehicleCosmetics/Wheels/Bugger_Inverted/Materials/MF_Wheel_Bugger_Inverted
graph {nodes [
  n1 = TextureCoordinate() {coordinateIndex=1}
  n2 = TextureSample(UVs=Panner(Coordinate=TextureCoordinate() {coordinateIndex=1, vTiling=7}) {speedY=6}) {samplerSource=SSM_Wrap_WorldGroupSettings, texture=/VehicleCosmetics/Wheels/_SharedTextures/FX/GradientStripe_T}
  n3 = Multiply(
   A=Multiply(
    A=TextureSample(UVs=@n1) {samplerSource=SSM_Wrap_WorldGroupSettings, texture=/VehicleCosmetics/Wheels/_SharedTextures/FX/Vert_Linear02_D}.RGBA
    B=Multiply(B=@n2.RGBA) {constA=2, constB=2}
   )
   B=TextureSample(UVs=@n1) {samplerSource=SSM_Wrap_WorldGroupSettings, texture=/VehicleCosmetics/Wheels/Bugger_Inverted/Textures/Rim_Bugger_Inverted_Masks}.RGBA
  )
  n4 = VectorParameter() {group=2_Rim Properties, parameterName=Custom Rim Color, sortPriority=-1, defaultValue=<rgba 0.7991 0 0.0052 1>}
  n5 = Multiply(
   A=AppendVector(A=@n4.RGB, B=@n4.RGB)
   B=ScalarParameter() {defaultValue=1, group=2_Rim Properties, parameterName=Emission Rim, sortPriority=0}
  )
  n6 = TextureSampleParameter2D() {
   group=1_Input Textures
   parameterName=Rim_Masks
   samplerSource=SSM_Wrap_WorldGroupSettings
   samplerType=SAMPLERTYPE_LinearColor
   sortPriority=0
   texture=/VehicleCosmetics/Wheels/Bugger_Inverted/Textures/T_BuggerInverted_Rim_M
  }
  n7 = Multiply(A=@n6.R, B=@n6.R)
  n8 = Sine(None=Time()) {period=3}
  n9 = LinearInterpolate(
   A=Multiply(A=@n7, B=ScalarParameter() {defaultValue=1, group=2_Rim Properties, parameterName=Rim Glow Pulse Max Intensity, sortPriority=1})
   B=Multiply(A=@n7, B=ScalarParameter() {defaultValue=0.5, group=2_Rim Properties, parameterName=Rim Glow Pulse Minimum Intensity, sortPriority=2})
   Alpha=Multiply(A=Add(A=@n8) {constB=2}) {constA=0.5, constB=0.5}
  )
  n10 = Add(
   A=Add(A=Multiply(B=@n3) {constA=0.3}, B=Multiply(A=@n3, B=@n5) {constA=0.3})
   B=Multiply(A=@n5, B=LinearInterpolate(A=@n6.R, B=@n9, Alpha=@n6.G))
  )
  n11 = VectorParameter() {group=3_Tire Properties, parameterName=Custom Tire Color, sortPriority=-1, defaultValue=<rgba 0.7991 0 0.0052 1>}
  n12 = AppendVector(A=@n11.RGB, B=@n11.RGB)
  n13 = ScalarParameter() {defaultValue=1, group=3_Tire Properties, parameterName=Emission Tire, sortPriority=0}
  n14 = TextureCoordinate()
  n15 = TextureSampleParameter2D(UVs=@n14) {
   group=1_Input Textures
   parameterName=Tire_Masks
   samplerSource=SSM_Wrap_WorldGroupSettings
   sortPriority=1
   texture=/VehicleCosmetics/Wheels/Bugger_Inverted/Textures/T_Bugger_Inverted_TIre_M2
  }
  n16 = Add(A=Multiply(A=Constant() {r=0.6}, B=@n15.G), B=@n15.R)
  n17 = TextureSample(UVs=Panner(Coordinate=@n14) {speedX=-1}) {samplerSource=SSM_Wrap_WorldGroupSettings, texture=/VehicleCosmetics/Wheels/Bugger_Inverted/Textures/Tread_Bugger_Inverted_RGB2}
  n18 = Add(A=TextureCoordinate(), B=Constant2Vector() {r=-0.2})
  n19 = Subtract(
   A=Divide(A=TextureSample(UVs=@n18) {samplerSource=SSM_Wrap_WorldGroupSettings, texture=/VehicleCosmetics/Wheels/_SharedTextures/FX/Hort_Linear02_D}.RGBA)
   B=TextureSample() {texture=&T_Bugger_Inverted_TIre_M2}.B
  )
  n20 = Multiply(A=LinearInterpolate(A=Multiply(A=@n17.R, B=@n19), Alpha=@n15.B) {constB=0}) {constB=3}
  n21 = Add(A=Multiply(A=Multiply(A=@n12, B=@n13), B=Add(A=@n16, B=@n20)), B=Multiply(A=@n20) {constB=0.3})
  n22 = VertexColor()
  FunctionOutput(None=SetMaterialAttributes!(MaterialAttributes, EmissiveColor=LinearInterpolate(A=@n10, B=@n21, Alpha=@n22.R)) {attributeSetTypes [#1]}) {outputName=MaterialAttributes}
 ]}
