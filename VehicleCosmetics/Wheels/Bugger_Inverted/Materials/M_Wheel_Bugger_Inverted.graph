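# Unlit material for the Bugger Inverted vehicle wheel cosmetic, set up for skeletal meshes and mesh deformers.
# The wheel-specific function MF_Wheel_Bugger_Inverted builds the base material attributes.
# Those attributes pass through the MF_BaseVehicle_Damage function, then into MF_VehicleDeRez_Transition for the de-rez transition.
# All surface logic lives in these chained material functions, with no other nodes in the graph.
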Material /VehicleCosmetics/Wheels/Bugger_Inverted/Materials/M_Wheel_Bugger_Inverted
graph {
 properties {
  bAutomaticallySetUsageInEditor=false
  bUseMaterialAttributes=true
  bUsedWithMeshDeformer=true
  bUsedWithSkeletalMesh=true
  previewMesh=/VehicleCosmetics/Wheels/Bugger_Inverted/Meshes/SK_Wheel_BuggerInverted
  shadingModel=MSM_Unlit
  usedShadingModels=Unlit
 }
 nodes [MaterialFunctionCall(In=MaterialFunctionCall(Material=MaterialFunctionCall() {materialFunction=&/MF_Wheel_Bugger_Inverted}.MaterialAttributes) {materialFunction=&MF_BaseVehicle_Damage}.Result) {materialFunction=&MF_VehicleDeRez_Transition}]
}
